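# Translucent unlit two-sided Niagara material for the flame energy effect of the GreaseClip vehicle boost, usable on meshes, ribbons and sprites.
# Panning wispy fire noise is masked by a UV gradient eroded with panning Perlin noise, plus a sine-shaped flame core band.
# The mask drives opacity, with particle alpha and optional Fresnel edge fade, and normal-based world position offset.
# A three-color gradient tinted by particle color feeds emissive.
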Material /VehicleCosmetics/Boosts/Boost_GreaseClip/FX/M_GreaseClip_FlamesEnergy
graph {
 properties {
  bAutomaticallySetUsageInEditor=false
  bUsedWithNiagaraMeshParticles=true
  bUsedWithNiagaraRibbons=true
  bUsedWithNiagaraSprites=true
  blendMode=BLEND_Translucent
  shadingModel=MSM_Unlit
  twoSided=true
  usedShadingModels=Unlit
 }
 nodes [
  n1 = OneMinus(None=ComponentMask(None=NamedRerouteUsage()) {r=true})
  n2 = LinearInterpolate(
   A=VectorParameter() {parameterName=ColorC, defaultValue=<rgba 1 0.0595 0.01 1>}.RGB
   B=VectorParameter() {parameterName=ColorB, defaultValue=<rgba 1.5 0.748 0.09 1>}.RGB
   Alpha=Power(Base=@n1) {constExponent=0.8}
  )
  n3 = VectorParameter() {parameterName=ColorA, defaultValue=<rgba 2 1.8333 1.6 1>}
  n4 = Saturate(None=Add(A=Power(Base=@n1) {constExponent=5}, B=NamedRerouteUsage()))
  n5 = ParticleColor()
  MaterialFunctionCall(Emissive=Multiply(A=LinearInterpolate(A=@n2, B=@n3.RGB, Alpha=@n4), B=@n5.RGB)) {materialFunction=&MF_FX_EmissiveAdjustments}
  n6 = AppendVector(A=ScalarParameter() {defaultValue=0.4, parameterName=Flames_ScaleU}, B=ScalarParameter() {defaultValue=3, parameterName=Flames_ScaleV})
  n7 = LinearInterpolate(
   A=NamedRerouteUsage()
   B=MaterialFunctionCall(UVs=NamedRerouteUsage(), Rotation Center=Constant(), "Rotation Angle (0-1)"=ScalarParameter() {parameterName=UV_Rotation_Angle}) {materialFunction=&CustomRotator}.Rotated Values
   Alpha=ScalarParameter() {parameterName=Use_CustomRotator}
  )
  n8 = ScalarParameter() {defaultValue=-1.33, parameterName=Flames_SpeedU}
  n9 = ScalarParameter() {parameterName=Flames_SpeedV}
  n10 = TextureSampleParameter2D(UVs=Panner(Coordinate=Multiply(A=@n7, B=@n6), Speed=AppendVector(A=@n8, B=@n9)) {speedX=-1.33}) {parameterName=Tex1, samplerType=SAMPLERTYPE_Grayscale, texture=/Game/Effects/Fort_Effects/Textures/Fire/T_WispyFireNoise_Soft_M}
  n11 = ComponentMask(None=NamedRerouteUsage()) {r=true}
  n12 = Saturate(None=Multiply(A=@n11, B=ScalarParameter() {defaultValue=15, parameterName=BottomFade_Hardness}) {constB=15})
  n13 = AppendVector(A=ScalarParameter() {defaultValue=0.1, parameterName=Distort_ScaleU}, B=ScalarParameter() {defaultValue=1, parameterName=Distort_ScaleV})
  n14 = ScalarParameter() {defaultValue=-1.55, parameterName=Distort_SpeedU}
  n15 = ScalarParameter() {parameterName=Distort_SpeedV}
  n16 = TextureSampleParameter2D(UVs=Panner(Coordinate=Multiply(A=@n7, B=@n13), Speed=AppendVector(A=@n14, B=@n15)) {speedX=-1.55}) {parameterName=Tex2, texture=/Game/Global/Textures/Default/Noise/3xPerlin_M}
  n17 = LinearInterpolate(A=@n16.R, B=@n16.R)
  n18 = Multiply(A=@n17) {constB=1.2}
  n19 = OneMinus(None=@n11)
  n20 = Multiply(A=Saturate(None=Subtract(A=@n12, B=@n18)), B=Saturate(None=Subtract(A=@n19, B=@n17)))
  n21 = NamedRerouteUsage()
  n22 = Saturate(None=Add(A=Multiply(A=@n10.R, B=@n20), B=Multiply(A=@n12, B=@n21)))
  Multiply(
   A=Multiply(A=@n5.A, B=@n22)
   B=LinearInterpolate(B=NamedRerouteUsage(), Alpha=ScalarParameter() {defaultValue=1, parameterName=Use_Fresnel}) {constA=1}
  )
  Multiply(A=Multiply(A=@n22, B=VertexNormalWS()), B=ScalarParameter() {defaultValue=5, parameterName=WPO_Scale}) {constB=5}
  n23 = TextureCoordinate()
  n24 = MaterialFunctionCall(XY=@n23) {materialFunction=&Swizzle}
  NamedRerouteDeclaration(None=LinearInterpolate(A=@n23, B=OneMinus(None=@n24.YX), Alpha=ScalarParameter() {parameterName=SwizzleUVs})) {name=UVs, nodeColor=<rgba 0 0.1328 1 1>}
  n25 = Power(Base=OneMinus(None=Fresnel() {baseReflectFraction=0, exponent=1}))
  NamedRerouteDeclaration(None=Saturate(None=Multiply(A=@n25) {constB=3})) {name=Fresnel, nodeColor=<rgba 1 0.5391 0 1>}
  n26 = Subtract(A=@n11, B=Multiply(A=@n17) {constB=0.5})
  n27 = ScalarParameter() {defaultValue=0.5, parameterName=FireBase_Offset}
  n28 = Sine(None=Multiply(A=Add(A=@n26, B=@n27) {constB=0.5}) {constB=0.5})
  n29 = ScalarParameter() {defaultValue=20, parameterName=FireBase_Power}
  NamedRerouteDeclaration(None=Multiply(A=Power(Base=@n28, Exp=@n29) {constExponent=20}) {constB=0.33}) {name=FlameCore, nodeColor=<rgba 1 0.4384 0.1084 1>}
 ]
}
