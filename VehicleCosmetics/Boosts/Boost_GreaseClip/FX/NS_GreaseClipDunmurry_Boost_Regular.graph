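# Niagara boost effect for the GreaseClip Dunmurry vehicle cosmetic, drawing orange thruster flames and sparks at socket positions plus per-nozzle offsets.
# ThrusterFlames bursts persistent particles that layer meshes and glow sprites, flickered by a waveform oscillation and enlarged by HyperBoostAlpha.
# GPU Sparks spray continuously in a cone, with curl noise, gravity-like pull and camera-distance size scaling.
# User.IsCinematic switches spark direction and bypasses the flames' camera fade.
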NiagaraSystem /VehicleCosmetics/Boosts/Boost_GreaseClip/FX/NS_GreaseClipDunmurry_Boost_Regular
dependencies {
 usedDataInterfaces [
  &NiagaraDataInterfaceCamera
  &NiagaraDataInterfaceCurve
  &NiagaraDataInterfaceSocketReader
  &NiagaraDataInterfaceArrayFloat3
 ]
 usedDynamicInputs [&Max_Float]
 usedModules [
  &self:SystemSpawnScript.SystemScriptSource.SystemScriptGraph.NiagaraNodeAssignment_0.SetVariables_#1
  &self:SystemSpawnScript.SystemScriptSource.SystemScriptGraph.NiagaraNodeAssignment_1.SetVariables_#2
  &self:SystemSpawnScript.SystemScriptSource.SystemScriptGraph.NiagaraNodeAssignment_2.SetVariables_#3
  &SystemState
  &self:SystemSpawnScript.SystemScriptSource.SystemScriptGraph.NiagaraNodeAssignment_3.SetVariables_#4
 ]
}
emitters {
 Sparks {
  simTarget=GPUComputeSim
  emitterUpdateScript [
   &EmitterState()
   &SpawnRate(SpawnRate=60, Use Spawn Probability=false)
   &SpawnBurst_Instantaneous(Spawn Count=&UniformRangedInt(Minimum=10, Maximum=15), Use Spawn Probability=false) {enabled=false}
  ]
  particleSpawnScript [
   &PartitionParticles(Partition Mode:ENiagara_PartitionMode_v2=::NewEnumerator1)
   &InitializeParticle(
    Color=<rgba 5 1.4016 0 1>
    Lifetime=&UniformRangedFloat(Minimum=0.1, Maximum=0.35)
    Sprite Size Max=<4 30>
    Sprite Size Min=<3 20>
    Sprite Size Mode=::NewEnumerator2
    UsePositionOffset=false
    Write Material Random=true
   ) {module=InitializeParticle001}
   &SocketLocation(
    SamplePosOffset=true
    Sampled Position Offset=&SelectVectorFromArray(Direct Array Index=$Particles.PartitionParticles.Index In Partition, Vector Selection Array=$System.ThrusterOffset)
    Socket Index=$Particles.PartitionParticles.Partition
    Socket Reader=$System.Socket Reader
    Socket Sampling Mode=::NewEnumerator3
    VelocitySpeedLimit=false
   )
   /Niagara/Modules/Spawn/Location/SphereLocation(
    Hemisphere X=true
    MaskSpawnGroup=false
    Module.Override Randomness=false
    Non Uniform Scale=<-1 1 1>
    Sphere Radius=&UniformRangedFloat(Minimum=10, Maximum=20)
    Surface Only=false
   )
   &AddVelocityFromPoint(
    Module.Use Falloff Distance=false
    Velocity Origin:Vector3f=&SimulationPosition()
    Velocity Strength=&UniformRangedFloat(Minimum=15, Maximum=60)
   )
   &AddVelocity(
    Cone Angle=10
    Cone Axis=&MakeCustomVectorFromBool(Bool=$User.IsCinematic, True Vector=<0 -1 0>, False Vector=<-1 0 0>)
    Module.UseConeFalloff=false
    Module.UseVelDistribution=false
    Override Seed=false
    Velocity Mode=::NewEnumerator2
    Velocity Speed=&RandomRangeFloat(Minimum=800, Maximum=1600)
   ) {module=AddVelocity001}
  ]
  particleUpdateScript [
   &ParticleState()
   /SetVariables(Particles.DistanceScale:NiagaraFloat=&ClampFloat(Float=&Divide_Float(
      A=&Subtract_Float(
       A=&VectorLength(Vector=&Subtract_Position(
         A=$Particles.Position
         B=&ReturnCameraPosition(Coordinate Space:ENiagaraCoordinateSpace=::Simulation, Camera Query:NiagaraDataInterfaceCamera=DataInterface())
        ))
       B=150
      )
      B=150
     ))) {isSetParameters=true}
   &AccelerationForce(Acceleration=<0 0 -250>)
   &CurlNoiseForce(Bias Noise=false, Noise Strength=550, Pan Noise=false)
   &SolveForcesAndVelocity(Clamp Velocity=false, Limit Acceleration=false)
   &ScaleColor(
    Scale Alpha=&FloatFromCurve(FloatCurve=DataInterface() {propertyValues {Curve {keys [{arriveTangent=-7.383, interpMode=RCIM_Cubic, leaveTangent=-7.383, tangentMode=RCTM_User, time=0, value=2}, <at 0.2 1 RCIM_Cubic>]}}})
    Scale RGB=&VectorFromFloat(Value=&FloatFromCurve(FloatCurve=DataInterface() {propertyValues {Curve {keys [<at 0 2 RCIM_Cubic>,<at 0.8 1 RCIM_Cubic>]}}}))
    ScaleA=true
    ScaleRGB=true
   )
   &ScaleSpriteSize(Scale Factor=&Vector2DFromFloat(Value=&FloatFromCurve(
      FloatCurve=DataInterface() {propertyValues {Curve {keys [
          {arriveTangent=50.3486, interpMode=RCIM_Cubic, leaveTangent=50.3492, tangentMode=RCTM_User, time=0, value=0}
          <at 0.03 1 RCIM_Cubic>
          {arriveTangent=-3.0852, interpMode=RCIM_Cubic, leaveTangent=-3.0852, tangentMode=RCTM_User, time=1, value=0}
         ]}}}
      Scale Curve=$Particles.DistanceScale
     )))
   &FN_Sprite_Attributes(
    Color=&MultiplyLinearColor(
     Linear Color=$Particles.Color
     Scale Factor=&MakeVector4_FromLinearColor(LinearColor=<rgba 0.5 0.5 0.5 0.03>)
    )
    Create Color Attribute=true
    Create SpriteSize Attribute=true
    SpriteSize=&Multiply_Vector2D(A=$Particles.SpriteSize, B=<10 3>)
    Write to Module Namespace=true
   )
   &DynamicMaterialParameters(
    Index 0 Param 1=1
    Index 0 Param 2=1
    Index 0 Param 3=1
    Module.Index0Float0Write=true
    Module.Index0Float1Write=true
    Module.Index0Float2Write=true
    Module.Index0Float3Write=true
   )
  ]
  rendererClasses [&NiagaraSpriteRendererProperties]
  renderers [
   {rendererClass=&NiagaraSpriteRendererProperties}
   {rendererClass=&NiagaraSpriteRendererProperties, rendererIndex=1}
  ]
 }
 ThrusterFlames {
  emitterSpawnScript [/SetVariables() {isSetParameters=true}]
  emitterUpdateScript [
   &EmitterState(
    Inactive Response=::NewEnumerator1
    Life Cycle Mode=::NewEnumerator1
    Loop Behavior=::NewEnumerator1
    Loop Duration Mode=::NewEnumerator1
    UseLoopDelay=false
   )
   &SpawnBurst_Instantaneous(Spawn Count=16, Use Spawn Probability=false)
   /SetVariables(Emitter.SmoothVelocity:Vector3f=/Niagara/DynamicInputs/Math/Clamp_Vector(
     Vector:Vector3f=&Lerp_Vector(
      A=$Emitter.SmoothVelocity
      B=&TransformVector(Vector=$Engine.Owner.Velocity, Source Space=::World, Destination Space=::Local)
      Alpha=0.06
     )
     Min:Vector3f=<0 0 0>
     Max:Vector3f=<1000 1000 1000>
    )) {isSetParameters=true}
   /SetVariables(Emitter.ForwardDot:NiagaraFloat=&ClampFloat(Float=&Divide_Float(
      A=&Subtract_Float(A=&MakeFloatFromVector(Vector=$Emitter.SmoothVelocity, Channel=::NewEnumerator0), B=50)
      B=500
     ))) {isSetParameters=true}
  ]
  particleSpawnScript [
   &PartitionParticles(Partition Mode:ENiagara_PartitionMode_v2=::NewEnumerator1)
   &InitializeParticle(Lifetime=0.2, Mesh Scale=<0.7 0.9 0.7>, Mesh Scale Mode=::NewEnumerator4, UsePositionOffset=false, Write Material Random=true) {module=InitializeParticle001}
   &InitialMeshOrientation(Mesh Orientation Mode=::NewEnumerator7, Rotation=<0 0 0.5>, Use Rotation Vector=true) {module=InitialMeshOrientation001}
  ]
  particleUpdateScript [
   &ParticleState(Kill Particles When Lifetime Has Elapsed=false)
   &ScaleMeshSize(Scale Factor=<-1 1 1>) {enabled=false, module=ScaleMeshSize001}
   /SetVariables(Particles.Oscillation:NiagaraFloat=&Add_Float(
     A=&Waveform(
      Waveform Count=::NewEnumerator1
      Global Phase=$Particles.Age
      "Waveform [1]"=::NewEnumerator0
      "[1] Frequency"=2.256
      "[1] Amplitude Min/Max"=<0 1>
      "[1] Scale"=0.5
      "[1] Bias"=1
      "Waveform [2] Blend"=::NewEnumerator2
      "[2] Input Position"=0
      "[2] Phase"=$Emitter.Age
      "[2] Phase Scale"=1
      "[2] Compound Frequency"=2
      "[2] Frequency"=7.2563
      "[2] Amplitude Min/Max"=<0 1>
      "[2] Scale"=0.5
      "[2] Bias"=1
      "[2] Exponent"=1
      "[2] Pulse Duration"=0.1
      "[2] Clip Triangle Min/Max"=<0 0>
      "[2] Random Seed"=0
     )
     B=0
    )) {isSetParameters=true}
   &SocketLocation(
    SamplePosOffset=true
    Sampled Position Offset=&AddVectorToPosition(
     Position=<24.5 0 0>
     Vector=&SelectVectorFromArray(Direct Array Index=$Particles.PartitionParticles.Index In Partition, Vector Selection Array=$System.ThrusterOffset)
    )
    Socket Index=$Particles.PartitionParticles.Partition
    Socket Reader=$System.Socket Reader
    Socket Sampling Mode=::NewEnumerator3
    VelocitySpeedLimit=false
   )
   &ScaleMeshSize(Scale Factor=&Multiply_Vector(
     A=<7 1 1>
     B=&Multiply_Vector(
      A=&Lerp_Vector(A=<0.2 1 1>, B=<1 1 1>, Alpha=$Particles.Oscillation)
      B=&Lerp_Vector(A=<1 1 1>, B=<3 1 1>, Alpha=$System.HyperBoostAlpha)
     )
    ))
   &SolveForcesAndVelocity(Clamp Velocity=false, Limit Acceleration=false)
   &DynamicMaterialParameters(
    Index 0 Param 1=&RandomRangeFloat(Minimum=0, Maximum=1)
    Index 0 Param 2=5
    Index 0 Param 3=5
    Module.Param0WriteEnabled=true
    Module.Param0WriteEnabled3=true
    Module.Param1WriteEnabled=true
    Module.Param1WriteEnabled3=true
    Module.Param2WriteEnabled=true
    Module.Param2WriteEnabled3=true
    Module.Param3WriteEnabled=true
    Module.Param3WriteEnabled3=true
    Write Parameter Index 3=true
   )
   &ScaleColor(
    Scale Alpha=&Lerp_Float(
     A=0
     B=1
     Alpha=&MakeCustomFloatFromBool(
      Bool=$User.IsCinematic
      True Float=1
      False Float=/Niagara/DynamicInputs/Camera/CameraFade(
       Camera Query:NiagaraDataInterfaceCamera=DataInterface()
       Fade Offset:NiagaraFloat=150
       Fade Distance:NiagaraFloat=100
       Use Owner Position:NiagaraBool=true
      )
     )
    )
    Scale RGB=&MakeVectorFromLinearColorRGB(Color=<rgba 5 1.1123 0.3639 1>)
    ScaleA=true
    ScaleRGB=true
   )
   &DynamicMaterialParameters(
    Index 0 Param 1=&RandomRangeFloat(Minimum=0, Maximum=1)
    Index 3 Param 1=3
    Index 3 Param 2=50
    Index 3 Param 3=1
    Module.Index0Float0Write=true
    Module.Index0Float1Write=true
    Module.Index0Float2Write=true
    Module.Index0Float3Write=true
    Module.Index3Float0Write=true
    Module.Index3Float1Write=true
    Module.Index3Float2Write=true
    Module.Index3Float3Write=true
    Write Parameter Index 3=::NewEnumerator1
   ) {module=DynamicMaterialParameters001}
   /NiagaraFN/Sprite/FN_SpriteAlignmentAndFacing(SpriteAlignment:Vector3f=<0 0 1>, SpriteFacing:Vector3f=<-1 0 0>)
   &FN_Sprite_Attributes(
    Create Camera Offset Attribute=true
    Create Position Attribute=true
    Create SpriteSize Attribute=true
    Position:NiagaraPosition=&AddVectorToPosition(
     Position=$Particles.Position
     Vector=&RotateVectorByQuaternion(Vector To Rotate=<-8 0 0>, Quaternion=$Output.SocketLocation.SampledRotation)
    )
    SpriteSize=&Multiply_Vector2D_ByFloat(
     Float=&Multiply_Float(
      A=&Lerp_Float(A=1, B=1.75, Alpha=$System.HyperBoostAlpha)
      B=&Lerp_Float(A=1, B=1.2, Alpha=$Particles.Oscillation)
     )
     Vector2D=<50 30>
    )
    Write to Module Namespace=true
   )
   &FN_Sprite_Attributes(
    Camera Offset=60
    Color=<rgba 1 0.261 0 1>
    Create Camera Offset Attribute=true
    Create Color Attribute=true
    Create Position Attribute=true
    Create SpriteSize Attribute=true
    Position:NiagaraPosition=&AddVectorToPosition(
     Position=$Particles.Position
     Vector=&RotateVectorByQuaternion(Vector To Rotate=<-10 0 0>, Quaternion=$Output.SocketLocation.SampledRotation)
    )
    SpriteSize=&Multiply_Vector2D_ByFloat(
     Float=&Multiply_Float(
      A=&Lerp_Float(A=1, B=1.2, Alpha=$Particles.Oscillation)
      B=&Lerp_Float(A=1, B=1.5, Alpha=$System.HyperBoostAlpha)
     )
     Vector2D=<50 50>
    )
    Write to Module Namespace=true
   ) {module=FN_Sprite_Attributes001}
   &FN_Sprite_Attributes(
    Color=<rgba 1 0.261 0 1>
    Create Color Attribute=true
    Create SpriteSize Attribute=true
    SpriteSize=&Multiply_Vector2D_ByFloat(
     Float=&Multiply_Float(
      A=&Lerp_Float(A=1, B=1.2, Alpha=$Particles.Oscillation)
      B=&Lerp_Float(A=1, B=2.5, Alpha=$System.HyperBoostAlpha)
     )
     Vector2D=<300 5>
    )
    Write to Module Namespace=true
   ) {module=FN_Sprite_Attributes002}
  ]
  rendererClasses [&NiagaraMeshRendererProperties, &NiagaraSpriteRendererProperties]
  renderers [
   {rendererClass=&NiagaraMeshRendererProperties, rendererIndex=0}
   {rendererClass=&NiagaraSpriteRendererProperties, rendererIndex=1}
   {rendererClass=&NiagaraSpriteRendererProperties, rendererIndex=2}
   {rendererClass=&NiagaraSpriteRendererProperties, rendererIndex=3}
   {rendererClass=&NiagaraMeshRendererProperties, rendererIndex=4}
  ]
 }
}
summary {
 systemName=NS_GreaseClipDunmurry_Boost_Regular
 emitters [
  {emitterName=ThrusterFlames, rendererClasses [&NiagaraMeshRendererProperties, &NiagaraSpriteRendererProperties]}
  {emitterName=Sparks, simTarget=GPUComputeSim, rendererClasses [&NiagaraSpriteRendererProperties]}
 ]
 userVariables [
  User.HyperBoostAlpha {type=NiagaraFloat, defaultValue=0, description=""}
  User.IsCinematic {type=NiagaraBool}
 ]
}
